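# Bike wheel material for vehicle cosmetics that adds an animated, tinted, emissive tire pattern to MF_WheelParent_Simple.
# Five rotating samples of Tire_Mask_Ani at offset UVs are composited through strips of a rotator mask texture, then combined with a panning sample using static mask channels.
# The result overrides base color and emissive, with vertex color red blending it over the parent.
# Vehicle damage and de-rez transition functions finish the chain.
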Material /VehicleCosmetics/Wheels/Bike/Materials/M_Wheel_Bike
graph {
 properties {
  bAutomaticallySetUsageInEditor=false
  bUseMaterialAttributes=true
  bUsedWithMeshDeformer=true
  bUsedWithSkeletalMesh=true
  usedShadingModels=Default Lit
 }
 nodes [
  Constant2Vector() {r=3}
  Constant2Vector() {r=4}
  n1 = TextureCoordinate()
  n2 = Add(A=@n1, B=Constant2Vector() {r=1})
  n3 = Constant() {r=0.2}
  n4 = Multiply(A=Add(A=@n1, B=Constant2Vector()), B=@n3)
  n5 = LinearInterpolate(
   A=TextureSample(UVs=Multiply(A=@n2, B=@n3)) {texture=/VehicleCosmetics/Wheels/Bike/Textures/T_Tire_Bike_RotatorMask}.RGB
   B=LinearInterpolate(B=TextureSample(UVs=@n4) {texture=&T_Tire_Bike_RotatorMask}.RGB, Alpha=Constant3Vector()) {constB=0}
  ) {constB=0}
  n6 = Constant2Vector() {r=2}
  n7 = TextureSample(UVs=Multiply(A=Add(A=@n1, B=@n6), B=@n3)) {texture=&T_Tire_Bike_RotatorMask}
  n8 = Multiply(A=Add(A=@n1, B=@n3))
  n9 = Add(A=@n1, B=@n3)
  LinearInterpolate(
   A=LinearInterpolate(
    A=LinearInterpolate(A=@n5, B=@n7.RGB) {constB=0}
    B=TextureSample(UVs=@n8) {texture=&T_Tire_Bike_RotatorMask}.RGB
   ) {constB=0}
   B=TextureSample(UVs=Multiply(A=@n9)) {texture=&T_Tire_Bike_RotatorMask}.RGB
   Alpha=Constant3Vector() {constant=<rgba 1 1 1 0>}
  ) {constB=0}
  n10 = MaterialFunctionCall() {materialFunction=&MF_WheelParent_Simple}
  n11 = GetMaterialAttributes(None=@n10.MaterialAttributes) {attributeGetTypes [#1, #2]}
  n12 = ScalarParameter() {defaultValue=1, group=3_Tire Properties, parameterName=Tire_Ani_Strength}
  n13 = VectorParameter() {group=3_Tire Properties, parameterName=Tire_Ani, channelNames {a=Unused, b="", g="", r=""}, defaultValue=<rgba 0.2435 1 0.2141 1>}
  n14 = TextureCoordinate()
  n15 = TextureSampleParameter2D(UVs=Rotator(Coordinate=Multiply(A=@n14)) {centerX=0.82, speed=1}) {
   desc=Mask animated
   group=1_Input Textures
   parameterName=Tire_Mask_Ani
   samplerSource=SSM_Wrap_WorldGroupSettings
   texture=/VehicleCosmetics/Textures/T_White
  }
  n16 = Constant2Vector() {r=1}
  n17 = Add(A=@n14, B=@n16)
  n18 = TextureSampleParameter2D(UVs=Rotator(Coordinate=Multiply(A=@n17)) {centerX=0.82, speed=1}) {
   desc=Mask animated
   group=1_Input Textures
   parameterName=Tire_Mask_Ani
   samplerSource=SSM_Wrap_WorldGroupSettings
   texture=&T_White
  }
  n19 = Multiply(A=Add(A=@n14, B=@n16)) {constB=0.2}
  n20 = TextureCoordinate()
  n21 = Constant2Vector() {r=2}
  n22 = Add(A=@n20, B=@n21)
  n23 = TextureSampleParameter2D(UVs=Rotator(Coordinate=Multiply(A=@n22)) {centerX=0.82, speed=1}) {
   desc=Mask animated
   group=1_Input Textures
   parameterName=Tire_Mask_Ani
   samplerSource=SSM_Wrap_WorldGroupSettings
   texture=&T_White
  }
  n24 = Add(A=@n20, B=@n21)
  n25 = LinearInterpolate(
   A=LinearInterpolate(A=@n15.RGBA, B=@n18.RGBA, Alpha=TextureSample(UVs=@n19) {texture=&T_Tire_Bike_RotatorMask}.RGB)
   B=@n23.RGBA
   Alpha=TextureSample(UVs=Multiply(A=@n24) {constB=0.2}) {texture=&T_Tire_Bike_RotatorMask}.RGB
  )
  n26 = TextureCoordinate()
  n27 = Constant2Vector() {r=3}
  n28 = Add(A=@n26, B=@n27)
  n29 = TextureSampleParameter2D(UVs=Rotator(Coordinate=Multiply(A=@n28)) {centerX=0.82, speed=1}) {
   desc=Mask animated
   group=1_Input Textures
   parameterName=Tire_Mask_Ani
   samplerSource=SSM_Wrap_WorldGroupSettings
   texture=&T_White
  }
  n30 = TextureSample(UVs=Multiply(A=Add(A=@n26, B=@n27)) {constB=0.2}) {texture=&T_Tire_Bike_RotatorMask}
  n31 = TextureCoordinate()
  n32 = Constant2Vector() {r=4}
  n33 = Add(A=@n31, B=@n32)
  n34 = TextureSampleParameter2D(UVs=Rotator(Coordinate=Multiply(A=@n33)) {centerX=0.82, speed=1}) {
   desc=Mask animated
   group=1_Input Textures
   parameterName=Tire_Mask_Ani
   samplerSource=SSM_Wrap_WorldGroupSettings
   texture=&T_White
  }
  n35 = Multiply(A=Add(A=@n31, B=@n32)) {constB=0.2}
  n36 = MaterialFunctionCall(Float4=LinearInterpolate(
    A=LinearInterpolate(A=@n25, B=@n29.RGBA, Alpha=@n30.RGB)
    B=@n34.RGBA
    Alpha=TextureSample(UVs=@n35) {texture=&T_Tire_Bike_RotatorMask}.RGB
   )) {materialFunction=&BreakOutFloat4Components}
  n37 = Multiply(A=TextureCoordinate(), B=ScalarParameter() {defaultValue=1, parameterName=Size})
  n38 = TextureSampleParameter2D(UVs=Panner(Coordinate=@n37) {speedX=0.35}) {
   desc=Mask animated
   group=1_Input Textures
   parameterName=Tire_Mask_Ani
   samplerSource=SSM_Wrap_WorldGroupSettings
   texture=&T_White
  }
  n39 = TextureSampleParameter2D(UVs=@n37) {
   desc=Mask animated
   group=1_Input Textures
   parameterName=Tire_Mask_Ani
   samplerSource=SSM_Wrap_WorldGroupSettings
   texture=&T_White
  }
  n40 = OneMinus(None=@n39.A)
  n41 = LinearInterpolate(A=@n38.B, Alpha=OneMinus(None=Multiply(A=@n39.G, B=@n40))) {constB=0}
  n42 = LinearInterpolate(B=@n36.A, Alpha=OneMinus(None=@n39.G)) {constA=1}
  n43 = LinearInterpolate(A=@n36.R, B=@n41, Alpha=@n42)
  n44 = Multiply(A=@n12, B=Multiply(A=@n13, B=@n43))
  n45 = OneMinus(None=@n42)
  n46 = Multiply(A=@n13, B=@n41)
  n47 = VectorParameter() {group=3_Tire Properties, parameterName=Tire_Ani_Base, channelNames {a=Unused, b="", g="", r=""}, defaultValue=<rgba 0.3741 0.1535 1 1>}
  n48 = Multiply(A=@n45, B=@n47)
  n49 = SetMaterialAttributes!(
   MaterialAttributes=@n11.MaterialAttributes
   Emissive Color=Add(A=LinearInterpolate(A=@n11.Emissive Color, B=@n44, Alpha=@n45) {constB=0}, B=Multiply(A=@n12, B=@n46))
   Base Color=LinearInterpolate(A=LinearInterpolate(A=@n11.Base Color, B=@n47, Alpha=@n45), B=LinearInterpolate(A=@n48, B=@n44, Alpha=@n43), Alpha=Add(A=@n45, B=@n41))
  ) {attributeSetTypes [#1, #2]}
  n50 = VertexColor()
  MaterialFunctionCall(In=MaterialFunctionCall(Material=BlendMaterialAttributes(A=@n10.MaterialAttributes, B=@n49, Alpha=@n50.R)) {materialFunction=&MF_BaseVehicle_Damage}.Result) {materialFunction=&MF_VehicleDeRez_Transition}
 ]
}
